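# Niagara system for the offhand trail of the female Knight Cat pickaxe.
# The one enabled emitter spawns a short-lived, bone/socket-driven ribbon through the FN AnimTrail modules, tinted by User.VariantTrailOffhand and faded by a low-alpha ScaleColor curve.
# Two emitters are disabled: a higher-alpha copy of the trail, and flickering sprite sparks spawned on User.SkeletalMesh with inherited velocity, drag and curl noise.
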NiagaraSystem /BRCosmetics/Weapons/FORT_Melee/Pickaxe_Knight_Cat_Female/FX/NS_KC_BI_Pickaxe_Trail_Offhand
dependencies {
 usedDataInterfaces [
  &NiagaraDataInterfaceColorCurve
  &NiagaraDataInterfaceCurve
  &NiagaraDataInterfaceCurlNoise
  &NiagaraDataInterfaceVectorCurve
 ]
 usedDynamicInputs []
 usedModules [&SystemState]
}
emitters {
 FN_AnimTrail_V4 {
  emitterUpdateScript [
   &EmitterState()
   /Game/Effects/Niagara/Modules/AnimTrail/FN_SpawnAnimTrail/FN_SpawnAnimTrail(Number of Anim Trails=1, SpawnRate=60)
  ]
  particleSpawnScript [
   /Game/Effects/Niagara/Modules/AnimTrail/FN_InitializeAnimTrail/FN_InitializeAnimTrailBoneIndecies(Second Bone or Socket Index=1)
   /Game/Effects/Niagara/Modules/AnimTrail/FN_InitializeAnimTrail/FN_InitializeAnimTrail(AnimTrailWidth=1, Color=$User.VariantTrailOffhand, Lifetime=0.1)
  ]
  particleUpdateScript [
   &SolveForcesAndVelocity(Clamp Velocity=false, Limit Acceleration=false)
   /Game/Effects/Niagara/Modules/AnimTrail/FN_UpdateAnimTrail/FN_UpdateAnimTrail(New float=1)
   /Game/Effects/Niagara/Modules/Materials/Material_v2()
   &ScaleColor(
    Linear Color Curve=DataInterface() {propertyValues {
      AlphaCurve {keys [<at 0 0.1>,<at 0.2124 0.05>,<at 0.5856 0>]}
      BlueCurve {keys [<at 0 1>,<at 0.2143 0.8>]}
      GreenCurve {keys [<at 0 1>,<at 0.2143 0.8>]}
      RedCurve {keys [<at 0 1>,<at 0.2143 0.8>]}
     }}
    Scale Mode=::NewEnumerator2
   ) {enabled=false}
   &ScaleColor(
    Linear Color Curve=DataInterface() {propertyValues {
      AlphaCurve {keys [<at 0 0.1>,<at 0.2124 0.05>,<at 0.5856 0>]}
      BlueCurve {keys [<at 0 1>,<at 0.2143 0.8>]}
      GreenCurve {keys [<at 0 1>,<at 0.2143 0.8>]}
      RedCurve {keys [<at 0 1>,<at 0.2143 0.8>]}
     }}
    Scale Mode=::NewEnumerator2
   ) {module=ScaleColor001}
  ]
  rendererClasses [&NiagaraRibbonRendererProperties]
  renderers [{rendererClass=&NiagaraRibbonRendererProperties}]
 }
 FN_AnimTrail_V4001 {
  bEnabled=false
  emitterUpdateScript [
   &EmitterState()
   &FN_SpawnAnimTrail(Number of Anim Trails=1, SpawnRate=60)
  ]
  particleSpawnScript [
   &FN_InitializeAnimTrailBoneIndecies(Second Bone or Socket Index=1)
   &FN_InitializeAnimTrail(AnimTrailWidth=1, Color=$User.VariantTrailOffhand, Lifetime=0.1)
  ]
  particleUpdateScript [
   &SolveForcesAndVelocity(Clamp Velocity=false, Limit Acceleration=false)
   &FN_UpdateAnimTrail(New float=1)
   &Material_v2()
   &ScaleColor(
    Linear Color Curve=DataInterface() {propertyValues {
      AlphaCurve {keys [<at 0 0.29>,<at 0.2124 0.13>,<at 0.5856 0>]}
      BlueCurve {keys [<at 0 1>,<at 0.2143 0.8>]}
      GreenCurve {keys [<at 0 1>,<at 0.2143 0.8>]}
      RedCurve {keys [<at 0 1>,<at 0.2143 0.8>]}
     }}
    Scale Mode=::NewEnumerator2
   )
  ]
  rendererClasses [&NiagaraRibbonRendererProperties]
  renderers [{rendererClass=&NiagaraRibbonRendererProperties}]
 }
 Sparks {
  bEnabled=false
  emitterUpdateScript [
   &EmitterState()
   &SpawnRate(SpawnRate=&RandomRangeFloat(Minimum=10, Maximum=20), Use Spawn Probability=true)
   &SpawnBurst_Instantaneous(Spawn Count=13, Use Spawn Probability=false, UseLoopCountLimit=false) {enabled=false}
  ]
  particleSpawnScript [
   &InitializeParticle(
    Color=<rgba 5 0 2.5 0.5>
    Lifetime=&RandomRangeFloat(Minimum=0.35, Maximum=0.5)
    Sprite Size=&RandomRangeVector2D(Minimum=<1 1.5>, Maximum=<2 3>)
    Sprite Size Mode=::NewEnumerator4
    UsePositionOffset=false
    Write Material Random=true
   )
   &SkeletalMeshLocation(
    MaskSpawnGroup=false
    Mesh Sampling Type=::NewEnumerator5
    Position Sampling=::NewEnumerator1
    Sampled Position Offset=&RandomRangeVector(Minimum=<0 0 0>, Maximum=<1 50 1>)
    Skeletal Mesh=$User.SkeletalMesh
    Socket Sampling Mode=::NewEnumerator5
   )
   &InheritVelocity(Inherited Velocity Amount Scale=<2 2 2>, Inherited Velocity Speed Limit=50, Speed Limit=true, UseSpeedThreshold=false) {module=InheritVelocity001}
  ]
  particleUpdateScript [
   &GravityForce(Gravity=<0 0 -25>)
   &Drag(Drag=&RandomRangeFloat(Minimum=10, Maximum=20), Use Linear Drag=true, Use Rotational Drag=false)
   &CurlNoiseForce(Noise Frequency=150, Noise Strength=25, Pan Noise=false)
   &SolveForcesAndVelocity(Clamp Velocity=false, Limit Acceleration=false)
   &ParticleState()
   &ScaleColor(
    Scale Alpha=&RandomRangeFloat(
     Minimum=&FloatFromCurve(FloatCurve=DataInterface() {propertyValues {Curve {keys [
          <at 0 1 RCIM_Cubic>
          {arriveTangent=-3.2218, interpMode=RCIM_Cubic, leaveTangent=-3.2218, tangentMode=RCTM_User, time=0.5, value=0}
          <at 0.6 0>
          <at 0.7 1>
          <at 0.8 0>
          <at 0.9 0.45>
          <at 1 0>
         ]}}})
     Maximum=&FloatFromCurve(FloatCurve=DataInterface() {propertyValues {Curve {keys [
          <at 0 0.35 RCIM_Cubic>
          {arriveTangent=-3.2218, interpMode=RCIM_Cubic, leaveTangent=-3.2218, tangentMode=RCTM_User, time=0.2, value=0.15}
          <at 0.4 -0.175>
          <at 0.5 0.225>
          <at 0.6 -0.175>
          <at 0.7 0.275>
          <at 1 0>
         ]}}})
    )
    Scale RGB=&VectorFromCurve(VectorCurve=DataInterface() {propertyValues {
       XCurve {keys [<at 0 1>,<at 0.6 0.45>,<at 0.7 0.75>,<at 0.8 0.3>,<at 0.9 0.55>,<at 1 0>]}
       YCurve {keys [<at 0 1>,<at 0.6 0.45>,<at 0.7 0.6>,<at 0.8 0.2>,<at 0.9 0.35>,<at 1 0>]}
       ZCurve {keys [<at 0 1>,<at 0.6 0.4>,<at 0.7 0.65>,<at 0.8 0.25>,<at 0.9 0.45>,<at 1 0>]}
      }})
    ScaleA=true
    ScaleRGB=true
   )
   &ScaleSpriteSize(Scale Factor=&Vector2DFromFloat(Value=&RandomRangeFloat(
      Minimum=&FloatFromCurve(FloatCurve=DataInterface() {propertyValues {Curve {keys [
           {arriveTangent=9.0725, interpMode=RCIM_Cubic, leaveTangent=9.0725, tangentMode=RCTM_User, time=0, value=0}
           {arriveTangent=7.5, interpMode=RCIM_Cubic, leaveTangent=7.5, time=0.1, value=1.25}
           {arriveTangent=-2.5249, interpMode=RCIM_Cubic, leaveTangent=-2.5249, tangentMode=RCTM_User, time=0.2, value=1.5}
           <at 0.6 0.05>
           <at 0.7 2>
           <at 0.8 0>
           <at 0.9 0.8>
           {arriveTangent=-0.0405, interpMode=RCIM_Cubic, leaveTangent=-0.0405, tangentMode=RCTM_User, time=1, value=0}
          ]}}})
      Maximum=&FloatFromCurve(FloatCurve=DataInterface() {propertyValues {Curve {keys [
           {arriveTangent=9.0725, interpMode=RCIM_Cubic, leaveTangent=9.0725, tangentMode=RCTM_User, time=0, value=0}
           {arriveTangent=9.375, interpMode=RCIM_Cubic, leaveTangent=9.375, time=0.1, value=1.625}
           {arriveTangent=-2.5249, interpMode=RCIM_Cubic, leaveTangent=-2.5249, tangentMode=RCTM_User, time=0.2, value=1.875}
           <at 0.4 0.125>
           <at 0.5 2.125>
           <at 0.6 0.125>
           <at 0.7 0.875>
           {arriveTangent=-0.0405, interpMode=RCIM_Cubic, leaveTangent=-0.0405, tangentMode=RCTM_User, time=0.9, value=0}
          ]}}})
     )))
  ]
  rendererClasses [&NiagaraSpriteRendererProperties]
  renderers [{rendererClass=&NiagaraSpriteRendererProperties}]
 }
}
summary {
 systemName=NS_KC_BI_Pickaxe_Trail_Offhand
 emitters [
  {emitterName=FN_AnimTrail_V4, rendererClasses [&NiagaraRibbonRendererProperties]}
  {bEnabled=false, emitterName=Sparks, rendererClasses [&NiagaraSpriteRendererProperties]}
  {bEnabled=false, emitterName=FN_AnimTrail_V4001, rendererClasses [&NiagaraRibbonRendererProperties]}
 ]
 userVariables [
  User.bIsFrontEndPreview {type=NiagaraBool}
  User.PlayerSpeed {type=NiagaraFloat, defaultValue=0, description=""}
  User.RibbonIDOffset {type=NiagaraInt32}
  User.SkeletalMesh {
   type=NiagaraDataInterfaceSkeletalMesh
   defaultValue {
    struct=NiagaraExt_VariableValue_DataInterface
    value {dataInterface=&self:NiagaraDataInterfaceSkeletalMesh_2}
   }
  }
  User.VariantTrailOffhand {type=LinearColor, description="", defaultValue {struct=LinearColor, value=<rgba 2 0 1 1>}}
 ]
}
